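# Additive two-sided material for Niagara mesh particles, drawing the wheel energy of the LightCycle skydiving contrail.
# A panning rainbow texture blends two color parameters, ColorChoice shifts toward the raw texture, and IntensityMask scales a V-based brightness boost.
# A second mask multiplies UV edge fades, a panning flow texture, particle alpha and inverted Fresnel.
# Niagara dynamic parameters drive ColorChoice, IntensityMask and the RandOffset shifting both panners.
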Material /BRCosmetics/Athena/Effects/Particles/Skydiving/NS_Contrail_LightCycle/Material/M_Contrail_Lightcycle_WheelEnergy
graph {
 properties {
  bAutomaticallySetUsageInEditor=false
  bUsedWithNiagaraMeshParticles=true
  blendMode=BLEND_Additive
  previewMesh=/BRCosmetics/Athena/Effects/Particles/Skydiving/NS_Contrail_LightCycle/SM_Contrail_Lightcycle_Energy7
  twoSided=true
  usedShadingModels=Default Lit
 }
 nodes [
  n1 = VectorParameter() {parameterName=Color1, defaultValue=<rgba 0.1633 0.6552 1 1>}
  n2 = VectorParameter() {parameterName=Color2, defaultValue=<rgba 1 0.5329 0.2384 1>}
  n3 = TextureCoordinate() {vTiling=2}
  n4 = DynamicParameter() {defaultValue=<rgba 0.5 1 1 1>, paramNames [ColorChoice, RandOffset, IntensityMask, Param4]}
  n5 = TextureSampleParameter2D!(UVs=Panner(Coordinate=Add(A=@n3, B=@n4.RandOffset)) {speedY=-0.5}, Level, Apply View MipBias) {group=Color, mipValueMode=TMVM_MipLevel, parameterName=Color Tex, texture=&T_Tile_RainbowColor}
  n6 = LinearInterpolate(A=LinearInterpolate(A=@n1.RGB, B=@n2.RGB, Alpha=@n5.RGB), B=@n5.RGB, Alpha=@n4.ColorChoice)
  n7 = ParticleColor()
  n8 = ComponentMask(None=TextureCoordinate()) {g=true}
  n9 = Multiply(A=Add(A=@n8) {constB=-0.4})
  n10 = Multiply(A=Add(A=OneMinus(None=@n8)) {constB=-0.3})
  n11 = Clamp(None=Multiply(A=Multiply(A=@n9, B=@n10), B=@n4.IntensityMask)) {maxDefault=10000}
  Add(A=Multiply(A=@n6, B=@n7.RGB), B=Multiply(A=@n7.RGB, B=Multiply(A=@n6, B=@n11)))
  n12 = TextureCoordinate()
  n13 = ComponentMask(None=@n12) {r=true}
  n14 = Multiply(A=@n13, B=OneMinus(None=@n13))
  n15 = ComponentMask(None=@n12) {g=true}
  n16 = Multiply(A=@n15, B=OneMinus(None=@n15))
  n17 = ScalarParameter() {defaultValue=1.3, parameterName=MaskPower}
  n18 = Multiply(A=Multiply(A=Power(Base=@n14), B=Power(Base=@n16, Exp=@n17) {constExponent=1.3})) {constB=50}
  n19 = TextureSampleParameter2D(UVs=Panner(Coordinate=Add(A=@n3, B=@n4.RandOffset)) {speedY=-1}) {
   group=Flow
   mipValueMode=TMVM_MipLevel
   parameterName=Flow Texture
   samplerType=SAMPLERTYPE_LinearColor
   texture=/BRCosmetics/Athena/Effects/Particles/Skydiving/NS_Contrail_Turtleneck_Crystal/T_WavePattern_Chromatic
  }
  Multiply(
   A=Multiply(A=Multiply(A=@n18, B=@n19.R), B=@n7.A)
   B=MaterialFunctionCall(Exponent=ScalarParameter() {defaultValue=1}) {materialFunction=&MF_Fresnel_FX}.Fresnel Inverted
  )
  SmoothStep(Min=@n1.RGB, Max=@n2.RGB, Value=@n5.RGB)
 ]
}
